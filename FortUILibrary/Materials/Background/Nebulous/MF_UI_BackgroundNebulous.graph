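# Material function for the animated Nebulous UI background, outputting a final color, a gradient color and left and right glow masks.
# One radial gradient yields both side glows by switching UV offsets at screen center, with an eased fade hiding the seam.
# Two counter-rotating particle grids draw triangle layers, and panning noise inside a soft rotated box forms a wind mask.
# The gradient is brightened and contrast-adjusted, with colors, opacities, rotation and offset exposed as inputs.
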MaterialFunction /FortUILibrary/Materials/Background/Nebulous/MF_UI_BackgroundNebulous
graph {nodes [
  n1 = ScreenPosition()
  n2 = AppendVector(A=Constant() {r=0.5}, B=Constant() {r=0.25})
  n3 = Add(A=@n1.ViewportUV, B=@n2)
  n4 = Subtract(A=@n1.ViewportUV, B=@n2)
  n5 = ComponentMask(None=@n1.ViewportUV) {r=true}
  n6 = NamedRerouteDeclaration(None=Round(None=NamedRerouteDeclaration(None=@n5) {name=Screen Gradient U, nodeColor=<rgba 0 0 0 1>})) {name=Side Mask, nodeColor=<rgba 0 0 0 1>}
  n7 = MaterialFunctionCall(
   UVs=MaterialFunctionCall!(ScreenUV=LinearInterpolate(A=@n3, B=@n4, Alpha=@n6), TextureSize, ScalePivotPoint) {materialFunction=/FortUILibrary/MaterialFunctions/Basics/MF_UI_ScreenSpaceUVScaling}.ScaleY
   Texture Scale=Constant2Vector() {g=4, r=3}
  ) {materialFunction=&ScaleUVsByCenter}
  n8 = MaterialFunctionCall("Value (0-1)"=Saturate(None=MaterialFunctionCall(UV=@n7.UVs) {materialFunction=&MF_UI_RadialGradient}.Smooth)) {materialFunction=&MF_UI_EaseCurves}
  n9 = NamedRerouteUsage()
  n10 = OneMinus(None=@n9)
  FunctionOutput(None=NamedRerouteDeclaration(None=Multiply(A=@n8.EaseInCubic, B=@n10)) {name=Left Glow Mask, nodeColor=<rgba 0 0 0 1>}) {outputName=Left Glow Mask, sortPriority=10}
  FunctionOutput(None=NamedRerouteDeclaration(None=Multiply(A=@n8.EaseInCubic, B=@n9)) {name=Right Glow Mask, nodeColor=<rgba 0 0 0 1>}) {outputName=Right Glow Mask, sortPriority=11}
  n11 = Add(A=Multiply(A=NamedRerouteUsage(), B=NamedRerouteUsage()), B=Multiply(A=NamedRerouteUsage(), B=NamedRerouteUsage()))
  n12 = Multiply(A=NamedRerouteUsage()) {constB=0.15}
  n13 = NamedRerouteUsage()
  n14 = Min(A=@n13, B=OneMinus(None=@n13))
  n15 = MaterialFunctionCall("Value (0-1)"=OneMinus(None=Multiply(A=@n14) {constB=2})) {materialFunction=&MF_UI_EaseCurves}
  n16 = FunctionOutput(None=LinearInterpolate(A=NamedRerouteUsage(), B=Saturate(None=@n11), Alpha=Multiply(A=@n12, B=@n15.EaseOutQuint) {desc=Avoid seeing the cut in the middle})) {outputName=Gradient Color, sortPriority=1}
  n17 = NamedRerouteUsage()
  n18 = Saturate(None=Add(A=@n16, B=Multiply(A=@n16, B=@n17)))
  n19 = Constant() {r=0.6}
  FunctionOutput(None=LinearInterpolate(
    A=@n18
    B=MaterialFunctionCall!(Color=@n18, Contrast=@n19) {materialFunction=/Engine/Functions/Engine_MaterialFunctions01/ImageAdjustment/Contrast_Preserve_Color}.Result
    Alpha=NamedRerouteUsage()
   )) {outputName=Final Color}
  n20 = Time()
  n21 = Multiply(A=@n20) {constB=0.1}
  n22 = Multiply(A=Add(A=@n20) {constB=0.25}) {constB=-0.05}
  Multiply(A=AppendVector(A=Sine(None=@n21), B=Cosine(None=@n22)), B=NamedRerouteUsage())
  NamedRerouteDeclaration(None=FunctionInput(Preview=Constant3Vector() {constant=<rgba 0.0319 1 0.2789 1>}) {bUsePreviewValueAsDefault=true, inputName=Secondary Color, sortPriority=1}) {name=Secondary Color, nodeColor=<rgba 0 0 0 1>}
  NamedRerouteDeclaration(None=FunctionInput(Preview=Constant3Vector() {constant=<rgba 0.0016 0.0573 0.0163 1>}) {bUsePreviewValueAsDefault=true, inputName=Primary Color}) {name=Primary Color, nodeColor=<rgba 0 0 0 1>}
  NamedRerouteDeclaration(None=FunctionInput(Preview=Constant3Vector() {constant=<rgba 0.0048 0.004 0.0116 1>}) {bUsePreviewValueAsDefault=true, inputName=Background Color, sortPriority=2}) {name=Background Color, nodeColor=<rgba 0 0 0 1>}
  NamedRerouteDeclaration(None=@n8.EaseInCubic) {name=Glows Mask, nodeColor=<rgba 0 0 0 1>}
  n23 = MaterialFunctionCall() {materialFunction=&MF_UI_ScreenSpaceUVScaling}
  NamedRerouteDeclaration(None=@n23.Scaled to Fill) {name=Scale to Fill Screenspace UV, nodeColor=<rgba 0 0 0 1>}
  n24 = Add(
   A=Rotator(
    Coordinate=NamedRerouteUsage()
    Time=FunctionInput() {bUsePreviewValueAsDefault=true, inputName=Rotation Multiplier, inputType=FunctionInput_Scalar, sortPriority=20}
   ) {speed=6.2832}
   B=AppendVector(
    A=Constant()
    B=FunctionInput() {bUsePreviewValueAsDefault=true, inputName=Offset Multiplier, inputType=FunctionInput_Scalar, sortPriority=30}
   )
  )
  n25 = NamedRerouteUsage()
  n26 = MaterialFunctionCall!(
   UV=Multiply(A=@n24) {constB=0.9}
   Global Offset
   Global Rotation=Multiply(A=Time()) {constB=0.005}
   Tiling=Constant() {r=2}
   Cell Offset=Constant2Vector()
   Cell Scale=LinearInterpolate(Alpha=@n25) {constA=-0.1, constB=0.1}
   Cell Rotation=LinearInterpolate(Alpha=@n25) {constA=-1}
   Random Seed=Constant() {r=12.8978}
  ) {materialFunction=/FortUILibrary/MaterialFunctions/Basics/MF_UI_ParticleGrid}
  NamedRerouteDeclaration(None=@n26.Random Cell Values) {name=Midground Triangles Random Cell Values, nodeColor=<rgba 0 0 0 1>}
  n27 = NamedRerouteUsage()
  n28 = MaterialFunctionCall(
   UV=@n24
   Global Rotation=Multiply(A=Time()) {constB=-0.005}
   Tiling=Constant() {r=3}
   Cell Offset=LinearInterpolate(Alpha=@n27) {constA=0.5}
   Cell Scale=LinearInterpolate(Alpha=@n27) {constA=-0.25, constB=0.1}
   Cell Rotation=LinearInterpolate(Alpha=@n27) {constA=-0.25, constB=0.25}
   Random Seed=Constant() {r=56}
  ) {materialFunction=&MF_UI_ParticleGrid}
  NamedRerouteDeclaration(None=@n28.Random Cell Values) {name=Background Triangles Random Cell Values, nodeColor=<rgba 0 0 0 1>}
  n29 = Subtract(A=NamedRerouteUsage(), B=AppendVector(A=Constant(), B=Constant() {r=0.35})) {constB=0.2}
  n30 = Rotator(Coordinate=@n29, Time=Constant() {r=0.05}) {speed=6.2832}
  n31 = Sine(None=Multiply(A=ComponentMask(None=@n29) {r=true}) {constB=0.5})
  n32 = AppendVector(A=Constant() {r=4}, B=LinearInterpolate(Alpha=Saturate(None=@n31)) {constA=1, constB=0.5})
  n33 = OneMinus(None=SmoothStep(Value=MaterialFunctionCall(UVs=@n30, Size=@n32) {materialFunction=&MF_UI_Box}.SDF) {constMax=0.25})
  n34 = MaterialFunctionCall(UVs=Panner(Coordinate=@n29) {speedX=-0.15}, Texture Scale=@n32) {materialFunction=&ScaleUVsByCenter}
  n35 = Constant() {r=0.05}
  n36 = NamedRerouteDeclaration(None=TextureSample(
    UVs=Rotator(Coordinate=@n34.UVs, Time=@n35) {speed=6.2832}
    Tex=TextureObject() {samplerType=SAMPLERTYPE_Alpha, texture=&/T_UI_Nebulous_Noise}
   ) {samplerSource=SSM_Wrap_WorldGroupSettings}.R) {name=Wind Noise, nodeColor=<rgba 0 0 0 1>}
  n37 = Saturate(None=MaterialFunctionCall("Value (0-1)"=Multiply(A=@n33, B=@n36)) {materialFunction=&MF_UI_EaseCurves}.EaseInCubic)
  n38 = FunctionInput(Preview=Constant() {r=0.3}) {bUsePreviewValueAsDefault=true, inputName=Wind Opacity, inputType=FunctionInput_Scalar, sortPriority=10}
  NamedRerouteDeclaration(None=Multiply(A=Multiply(A=@n37) {constB=0.75}, B=NamedRerouteDeclaration(None=@n38) {name=Wind Opacity, nodeColor=<rgba 0 0 0 1>})) {name=Wind Mask, nodeColor=<rgba 0 0 0 1>}
  n39 = Constant() {r=1}
  n40 = Constant() {r=3}
  n41 = MaterialFunctionCall!(UVs=@n28.Cells UV, Size=Constant() {r=0.5}, Sides=Constant() {r=3}, InnerAngle, OutlineThickness, OutlineAlignment, GlowMin, GlowMax) {materialFunction=/FortUILibrary/MaterialFunctions/Shapes/MF_UI_RegularPolygon}
  n42 = Add(
   A=NamedRerouteDeclaration(None=MaterialFunctionCall(UVs=@n26.Cells UV, Size=@n39, Sides=@n40) {materialFunction=&MF_UI_RegularPolygon}.Fill) {name=Triangle Midground, nodeColor=<rgba 0 0 0 1>}
   B=Multiply(A=NamedRerouteDeclaration(None=@n41.Fill) {name=Triangle Background, nodeColor=<rgba 0 0 0 1>}) {constB=0.75}
  )
  n43 = FunctionInput() {bUsePreviewValueAsDefault=true, inputName=Triangle Opacity, inputType=FunctionInput_Scalar, sortPriority=20, previewValue=<1 0 0 1>}
  NamedRerouteDeclaration(None=Multiply(A=Saturate(None=@n42), B=@n43)) {name=Triangles Mask, nodeColor=<rgba 0 0 0 1>}
  NamedRerouteDeclaration(None=@n23.Scaled to Fit) {name=Scale to Fit Screenspace UV, nodeColor=<rgba 0 0 0 1>}
 ]}
