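# Translucent UI material that draws a two-color menu background, either full or as a faded overlay.
# The UseNewBg static switch, on by default, picks the MF_UI_Background_Immersive function over the older MF_UI_BackgroundNebulous with its wind and triangle opacities.
# Display as Overlay swaps solid coverage for a rotatable, sine-eased linear fade scaled by Mask Fade Power.
# An adjustable top gradient mask is built from remapped screen-space Y.
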Material /FortUILibrary/Materials/Background/Nebulous/M_UI_Background_Nebulous
graph {
 properties {
  bUsedWithStaticMesh=false
  blendMode=BLEND_Translucent
  materialDomain=MD_UI
  pixelDepthOffsetMode=PDOM_Legacy
  usedShadingModels=Default Lit
 }
 nodes [
  n1 = MaterialFunctionCall() {materialFunction=&GetUserInterfaceUV}
  n2 = ScalarParameter() {parameterName=Overlay Rotation, sliderMax=1}
  n3 = MaterialFunctionCall("Value (0-1)"=ComponentMask(None=Rotator(Coordinate=@n1.Normalized UV, Time=@n2) {speed=6.2832}) {g=true}) {materialFunction=&MF_UI_EaseCurves}
  n4 = ScalarParameter() {defaultValue=1, parameterName=Mask Fade Power}
  LinearInterpolate(
   B=StaticSwitchParameter(True=Multiply(A=@n3.EaseOutSine, B=@n4), False=Constant() {r=1}) {parameterName=Display as Overlay}
   Alpha=NamedRerouteUsage()
  )
  n5 = VectorParameter() {parameterName=Primary Background Color, defaultValue=<rgba 0.0015 0.0001 0.0052 1>}
  n6 = VectorParameter() {parameterName=Secondary Background Color, defaultValue=<rgba 0.0545 0 0.2831 1>}
  n7 = ScalarParameter() {defaultValue=1, parameterName=Wind Opacity, sliderMax=1}
  n8 = ScalarParameter() {defaultValue=1, parameterName=Triangle Opacity, sliderMax=1}
  NamedRerouteDeclaration(None=StaticSwitchParameter(
    True=MaterialFunctionCall!(Primary Color=@n5.RGB, Background Color=@n6.RGB, Is Flipped Horizontally) {materialFunction=&/MF_UI_Background_Immersive}.Final Color
    False=MaterialFunctionCall!(
     Primary Color=@n5.RGB
     Secondary Color=@n6.RGB
     Background Color
     Wind Opacity=@n7
     Rotation Multiplier
     Triangle Opacity=@n8
     Offset Multiplier
    ) {materialFunction=&/MF_UI_BackgroundNebulous}.Final Color
   ) {defaultValue=true, parameterName=UseNewBg}) {name=Background Color, nodeColor=<rgba 0 0 0 1>}
  n9 = MaterialFunctionCall() {materialFunction=&GetUserInterfaceUV}
  n10 = ScalarParameter() {parameterName=Top Gradient Control}
  n11 = Saturate(None=MaterialFunctionCall(Input=ComponentMask(None=@n9.Normalized UV) {g=true}, Input Low=Constant(), Input High=@n10) {materialFunction=&RemapValue_0-1}.Result)
  n12 = MaterialFunctionCall("Value (0-1)"=@n10) {materialFunction=&MF_UI_EaseCurves}
  NamedRerouteDeclaration(None=Saturate(None=Subtract(A=@n11, B=@n12.EaseInSine))) {name=Top Gradient Mask, nodeColor=<rgba 0 0 0 1>}
  NamedRerouteUsage()
 ]
}
