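# UI alpha-composite material that draws an animated circular nebulous glow with particles.
# A rotating, angle-tapered circle outline forms the radial glow; two counter-scaled, rotating triangle-texture layers form particles, and a time-rotated cosine gradient supplies color.
# Noise panned in polar coordinates, divided by distance from a ring, forms the glow mask, sampled with unwarped-UV derivatives for mip selection.
# Named reroutes combine these layers, with opacity shaped by circle SDF and cubic easing.
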Material /FortUILibrary/Materials/Background/Nebulous/M_UI_Circle_Nebulous
graph {
 properties {
  bUsedWithStaticMesh=false
  blendMode=BLEND_AlphaComposite
  materialDomain=MD_UI
  pixelDepthOffsetMode=PDOM_Legacy
  usedShadingModels=Default Lit
 }
 nodes [
  n1 = Multiply(A=NamedRerouteUsage(), B=NamedRerouteUsage())
  Desaturation(None=@n1)
  Multiply(A=@n1, B=VertexColor().A)
  n2 = Multiply(A=Time()) {constB=0.2}
  n3 = MaterialFunctionCall!(UVs, Rotation=OneMinus(None=Frac(None=@n2))) {materialFunction=/FortUILibrary/MaterialFunctions/Gradients/MF_UI_AngleGradient}
  n4 = MaterialFunctionCall(R=Constant() {r=0.025}, G=Constant() {r=0.1}, B=Constant() {r=1.1}) {materialFunction=&AppendMany}
  n5 = MaterialFunctionCall(R=Constant(), G=Constant(), B=Constant() {r=0.5}) {materialFunction=&AppendMany}
  n6 = NamedRerouteUsage()
  n7 = NamedRerouteUsage()
  NamedRerouteDeclaration(None=LinearInterpolate(
    B=MaterialFunctionCall!(Gradient=@n3.Result, a=@n4.RGB, b=@n5.RGB, c, d) {materialFunction=/FortUILibrary/MaterialFunctions/Color/MF_UI_CosineGradient}.Result
    Alpha=LinearInterpolate(A=@n6, Alpha=@n7)
   )) {name=Glow Colors, nodeColor=<rgba 0.2656 1 0 1>}
  n8 = Time()
  n9 = ScalarParameter() {defaultValue=-2, parameterName=RotationSpeed}
  n10 = MaterialFunctionCall!(Vector or UVs=Rotator(
    Coordinate=MaterialFunctionCall() {materialFunction=&GetUserInterfaceUV}.Normalized UV
    Time=Multiply(A=@n8, B=@n9) {constB=-1}
   ) {speed=1}) {materialFunction=/FortUILibrary/MaterialFunctions/Basics/MF_UI_CheaperVectorToRadial}
  n11 = SmoothStep(Value=@n10.Vector Converted to Angle) {constMax=0, constMin=0.95}
  n12 = MaterialFunctionCall(
   Size=ScalarParameter() {defaultValue=0.666, parameterName=GlowSize}
   GlowMin=Multiply(A=OneMinus(None=@n10.Vector Converted to Angle), B=ScalarParameter() {defaultValue=-0.125, parameterName=GlowMin})
   GlowMax=Multiply(A=@n10.Vector Converted to Angle, B=ScalarParameter() {defaultValue=0.333, parameterName=GlowMax})
  ) {materialFunction=&MF_UI_Circle}
  NamedRerouteDeclaration(None=Saturate(None=Multiply(A=@n11, B=@n12.Outline))) {name=RadialGlow, nodeColor=<rgba 1 0 0.375 1>}
  n13 = MaterialFunctionCall() {materialFunction=&GetUserInterfaceUV}
  n14 = Rotator(Coordinate=@n13.Normalized UV) {speed=-0.1}
  n15 = Constant() {r=0.2}
  n16 = Rotator(Coordinate=@n13.Normalized UV) {speed=-0.25}
  n17 = Constant() {r=0.5}
  n18 = Add(
   A=TextureSample(UVs=MaterialFunctionCall(UVs=@n14, Texture Scale=@n15) {materialFunction=&ScaleUVsByCenter}.UVs) {samplerType=SAMPLERTYPE_LinearColor, texture=/Game/UI/Foundation/Items/Textures/T-Card-Hero-Anim-Triangle}.G
   B=TextureSample(UVs=MaterialFunctionCall(UVs=@n16, Texture Scale=@n17) {materialFunction=&ScaleUVsByCenter}.UVs) {samplerType=SAMPLERTYPE_LinearColor, texture=&T-Card-Hero-Anim-Triangle}.R
  )
  NamedRerouteDeclaration(None=Multiply(A=Saturate(None=@n18), B=NamedRerouteUsage())) {name=Particles, nodeColor=<rgba 0 0.3438 1 1>}
  n19 = MaterialFunctionCall() {materialFunction=&MF_UI_Circle}
  n20 = LinearInterpolate(A=NamedRerouteUsage(), B=NamedRerouteUsage(), Alpha=SmoothStep(Value=OneMinus(None=@n19.SDFOutline)) {constMax=1.5, constMin=0.4})
  NamedRerouteDeclaration(None=Saturate(None=MaterialFunctionCall("Value (0-1)"=@n20) {materialFunction=&MF_UI_EaseCurves}.EaseInCubic)) {name=Final Opacity, nodeColor=<rgba 1 0 0.5859 1>}
  n21 = MaterialFunctionCall() {materialFunction=&GetUserInterfaceUV}
  n22 = MaterialFunctionCall(Float2=MaterialFunctionCall(Vector or UVs=@n21.Normalized UV) {materialFunction=&MF_UI_CheaperVectorToRadial}.Radial Coordinates) {materialFunction=&BreakOutFloat2Components}
  n23 = Multiply(A=@n22.R)
  n24 = Multiply(A=@n22.G) {constB=2}
  n25 = ScalarParameter() {defaultValue=-0.05, parameterName=Perpendicular Speed}
  n26 = ScalarParameter() {defaultValue=-0.1, parameterName=Parallel Speed}
  n27 = MaterialFunctionCall() {materialFunction=&GetUserInterfaceUV}
  n28 = TextureSample!(
   UVs=Panner(Coordinate=AppendVector(A=@n23, B=@n24), Speed=AppendVector(A=@n25, B=@n26))
   Tex
   "DDX(UVs)"=DDX(Value=@n27.Normalized UV)
   "DDY(UVs)"=DDY(Value=@n27.Normalized UV)
   Apply View MipBias
  ) {mipValueMode=TMVM_Derivative, samplerType=SAMPLERTYPE_Grayscale, texture=/FortUILibrary/UIKit/Buttons/Textures/LowResBlurredNoise}
  n29 = Constant() {r=0.7}
  n30 = Multiply(A=@n29) {constB=-1}
  n31 = Constant() {r=-1}
  n32 = Constant() {r=1}
  n33 = ConstantBiasScale(None=@n21.Normalized UV) {bias=-0.5, scale=2}
  n34 = ComponentMask(None=ConstantBiasScale(None=Length(None=@n33) {is=class MaterialExpressionLength}) {bias=-0.5, scale=2}) {r=true}
  n35 = Max(A=Abs(None=MaterialFunctionCall(Input=@n34, Input Low=@n30, Input High=@n29, Target Low=@n31, Target High=@n32) {materialFunction=&RemapValueRange}.Result)) {constB=0.01}
  NamedRerouteDeclaration(None=Multiply(A=@n28.R, B=Divide(B=@n35))) {name=Glow Mask, nodeColor=<rgba 1 0.1875 0 1>}
 ]
}
